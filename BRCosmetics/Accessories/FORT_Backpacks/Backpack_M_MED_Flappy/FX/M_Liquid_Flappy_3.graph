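# Masked, two-sided liquid material for Niagara mesh particles on the Flappy backpack, faking a partly filled container.
# Particle dynamic parameters supply water plane, velocity, dimensions and fill; rotated local coordinates plus ripple noise form a water-line gradient that clips opacity.
# Virtual-plane coordinates map turbulent normals onto the fake surface, alongside foam, meniscus and parallax bubbles.
# Skylight environment sampling fakes tinted refraction, with sunlight-driven emissive and a camera-offset WPO zeroed in shadows.
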Material /BRCosmetics/Accessories/FORT_Backpacks/Backpack_M_MED_Flappy/FX/M_Liquid_Flappy_3
graph {
 properties {bUsedWithNiagaraMeshParticles=true, blendMode=BLEND_Masked, pixelDepthOffsetMode=PDOM_Legacy, twoSided=true, usedShadingModels=Default Lit}
 nodes [
  n1 = NamedRerouteUsage()
  Abs(None=@n1)
  Abs()
  Abs()
  n2 = NamedRerouteUsage()
  n3 = Multiply(
   A=NamedRerouteUsage()
   B=LinearInterpolate(
    A=ScalarParameter() {parameterName=Foam Min}
    B=ScalarParameter() {defaultValue=0.4, parameterName=Foam Max}
    Alpha=Power(Base=@n2) {constExponent=3}
   ) {constA=10, constB=5}
  )
  n4 = Multiply(A=ComponentMask(None=NamedRerouteUsage()) {b=true}, B=@n3) {constB=10}
  Add(A=@n4, B=NamedRerouteUsage())
  n5 = NamedRerouteUsage()
  n6 = Constant2Vector() {g=0.6, r=0.8}
  n7 = ScalarParameter() {parameterName=Viscosity}
  n8 = Multiply(
   A=Multiply(A=Multiply(A=@n5, B=@n6), B=LinearInterpolate(Alpha=@n7) {constA=1, constB=0.5})
   B=LinearInterpolate(Alpha=NamedRerouteUsage()) {constA=0.5}
  ) {constB=1.2}
  n9 = NamedRerouteUsage()
  n10 = Subtract(A=TextureSample(UVs=Panner(Coordinate=@n8, Time=@n9) {speedY=-2}) {samplerType=SAMPLERTYPE_LinearColor, texture=&HighContrastNoise}.RGB) {constB=0.3}
  n11 = Power(Base=NamedRerouteUsage())
  n12 = Multiply(A=Multiply(A=Multiply(A=@n10, B=@n11), B=NamedRerouteUsage())) {constB=0.75}
  Add(B=@n12) {constB=0}
  n13 = NamedRerouteUsage()
  Ceil(None=@n13)
  Ceil()
  ComponentMask(None=VectorParameter() {parameterName=Dimensions_Fill, defaultValue=<rgba 160 160 160 0.5>}.RGB) {b=true, g=true, r=true}
  Constant3Vector() {constant=<rgba 1 0 0 0>}
  Constant() {r=1}
  DotProduct(A=VertexNormalWS(), B=CameraVectorWS())
  n14 = NamedRerouteUsage()
  n15 = NamedRerouteUsage()
  n16 = Multiply(A=NamedRerouteUsage(), B=ScalarParameter() {defaultValue=5, parameterName=Water Surface Normal Tile})
  n17 = Multiply(
   A=LinearInterpolate(Alpha=NamedRerouteUsage()) {constA=1.2, constB=0.8}
   B=LinearInterpolate(Alpha=ScalarParameter() {parameterName=Viscosity}) {constA=1, constB=1.5}
  )
  n18 = MaterialFunctionCall(
   Plane Normal Axis=NamedRerouteUsage()
   Plane Center=Subtract(A=ParticlePositionWS(), B=Add(A=@n14, B=@n15))
   UV Scale=ComponentMask(None=Multiply(A=@n16, B=@n17)) {g=true, r=true}
  ) {materialFunction=&VirtualPlaneCoordinates}
  n19 = NamedRerouteUsage()
  n20 = Panner(Coordinate=@n18.UVs, Time=@n19) {speedY=0.4}
  n21 = TextureSample(UVs=@n20) {samplerType=SAMPLERTYPE_LinearColor, texture=/FortWater/Textures/02_WaterFoam_Veroni_base_H}
  n22 = Panner(
   Coordinate=MaterialFunctionCall!(Angle=Constant() {r=0.25}, UV/2DVector=@n18.UVs) {materialFunction=/Game/Characters/Player/Male/Medium/Bodies/M_MED_Faux/Materials/MF_2DRotate}.Result
   Time=@n19
  ) {speedX=0.4}
  n23 = TextureSample(UVs=@n22) {samplerType=SAMPLERTYPE_LinearColor, texture=&02_WaterFoam_Veroni_base_H}
  LinearInterpolate(A=@n21.G, Alpha=@n23.G)
  n24 = Constant() {r=0.013}
  LinearInterpolate(A=@n24, Alpha=NamedRerouteUsage()) {constA=0.05, constB=0.25}
  n25 = DotProduct(A=PixelNormalWS(), B=AtmosphericLightVector())
  n26 = LinearInterpolate(Alpha=Multiply(A=Add(A=@n25) {constB=-1}) {constB=-0.5}) {constA=1, constB=0.15}
  n27 = Saturate(None=Multiply(A=@n25) {constB=-1})
  n28 = AtmosphericLightColor()
  n29 = LinearInterpolate(A=Multiply(A=Power(Base=@n27) {constExponent=5}, B=@n28) {constB=2}, B=@n28, Alpha=NamedRerouteUsage())
  n30 = Multiply(A=Multiply(A=Add(A=@n26, B=@n29), B=NamedRerouteUsage()), B=ScalarParameter() {defaultValue=0.5, parameterName=Emissive Multiplier})
  MaterialFunctionCall(Emissive=@n30) {materialFunction=&MF_FX_EmissiveAdjustments}
  MaterialFunctionCall() {materialFunction=/Engine/Functions/GetPostProcessFunctions/GetAmbientCubemapIntensity}
  MaterialFunctionCall() {materialFunction=&CameraDirectionVector}
  Multiply(A=@n21.G, B=@n23.G)
  n31 = NamedRerouteUsage()
  n32 = Constant3Vector() {constant=<rgba 1 0 0 0>}
  n33 = Constant3Vector() {constant=<rgba 0 -1 0 0>}
  Multiply(B=MaterialFunctionCall(X=DotProduct(A=@n31, B=@n32), Y=DotProduct(A=@n31, B=@n33)) {materialFunction=&MakeFloat2}.Result) {constA=0.2}
  Multiply(A=NamedRerouteUsage()) {constB=0.1}
  n34 = NamedRerouteUsage()
  n35 = Multiply(A=NamedRerouteUsage()) {constB=3}
  n36 = Sine(None=Add(A=Multiply(A=@n34) {constB=2}, B=@n35) {constB=0})
  n37 = Sine(None=Add(A=Multiply(A=@n34) {constB=3}, B=Multiply(A=@n35) {constB=-1}) {constB=0})
  Multiply(A=Add(A=Multiply(A=@n36, B=@n1), B=Multiply(A=@n37, B=@n1))) {constB=0.01}
  n38 = DynamicParameter() {desc=Velocity, parameterIndex=1, defaultValue=<rgba 0 0 1 0>}
  NamedRerouteDeclaration(None=@n38.A) {name=Speed, nodeColor=<rgba 0 0.7891 1 1>}
  NamedRerouteDeclaration(None=Saturate(None=TwoSidedSign())) {name=Backface, nodeColor=<rgba 1 0 0.7031 1>}
  n39 = DynamicParameter() {desc=WaterPlaneTangent, parameterIndex=2, defaultValue=<rgba 1 0 0 0>}
  NamedRerouteDeclaration(None=MaterialFunctionCall(X=@n39.R, Y=@n39.R, Z=@n39.R) {materialFunction=&MakeFloat3}.Result) {name=WaterPlaneTangent, nodeColor=<rgba 0.8984 1 0 1>}
  n40 = Subtract(A=WorldPosition() {worldPositionShaderOffset=WPT_ExcludeAllShaderOffsets}.XYZ, B=ParticlePositionWS())
  n41 = NamedRerouteUsage()
  n42 = Constant3Vector() {constant=<rgba -1 -1 1 0>}
  n43 = MaterialFunctionCall!(VectorToRotate=@n40, LookAtVector=Normalize(VectorInput=Multiply(A=@n41, B=@n42)), RestingVector=Constant3Vector() {constant=<rgba 0 0 1 0>}) {materialFunction=/Engine/Functions/Engine_MaterialFunctions01/Vectors/RotateVector}
  n44 = Divide(A=@n43.Result, B=NamedRerouteUsage()) {constB=1}
  n45 = ComponentMask(None=@n44) {b=true}
  n46 = LinearInterpolate(Alpha=NamedRerouteUsage()) {constA=0.5, constB=-0.5}
  n47 = Multiply(A=Normalize(VectorInput=@n44), B=NamedRerouteUsage())
  n48 = Multiply(A=Power(Base=MaterialFunctionCall(Vector 2=@n47) {materialFunction=/Engine/Functions/Engine_MaterialFunctions02/Utility/VectorLength}.V2 Length)) {constB=0.5}
  n49 = NamedRerouteUsage()
  n50 = MaterialFunctionCall(Color=Add(A=@n12) {constB=0}) {materialFunction=&SplitComponents}
  n51 = Add(A=Multiply(A=Cosine(None=@n48), B=Multiply(A=@n49) {constB=-0.5}) {constB=0.15}, B=Add(A=@n50.R, B=@n50.R)) {constB=0}
  n52 = Add(A=Subtract(A=Multiply(B=@n45) {constA=-1}, B=@n46), B=@n51) {constB=0}
  n53 = Multiply(A=@n13, B=NamedRerouteDeclaration(None=@n52) {name=Water Line Gradient, nodeColor=<rgba 1 0 0.1172 1>}) {constA=1}
  NamedRerouteDeclaration(None=SmoothStep(Value=@n53) {constMax=0.001}) {name=Opacity, nodeColor=<rgba 0.4297 1 0 1>}
  n54 = ParticlePositionWS()
  n55 = NamedRerouteUsage()
  n56 = NamedRerouteUsage()
  NamedRerouteDeclaration(None=MaterialFunctionCall!(Position=Subtract(A=@n54, B=@n55), UVRadialScale=Divide(B=@n56)) {materialFunction=/Game/Weapons/WeaponSkins/MaterialFunction/ObjectStabilizedScreenSpaceUVs}.StabilizedUVs) {name=ScreenSpaceUV, nodeColor=<rgba 1 0.9375 0 1>}
  n57 = DotProduct(A=Constant3Vector() {constant=<rgba 0 0 -1 0>}, B=NamedRerouteUsage())
  NamedRerouteDeclaration(None=Multiply(A=Add(A=@n57)) {constB=0.5}) {name=WaterUpsideDown, nodeColor=<rgba 0 0.1797 1 1>}
  n58 = Abs(None=MaterialFunctionCall(Vector 3=NamedRerouteUsage()) {materialFunction=&VectorLength}.V3 Length)
  n59 = Constant()
  n60 = Constant() {r=100}
  NamedRerouteDeclaration(None=Saturate(None=MaterialFunctionCall(Input=@n58, Input Low=@n59, Input High=@n60) {materialFunction=&RemapValueRange}.Result)) {name=Temp, nodeColor=<rgba 1 0.9844 0 1>}
  NamedRerouteDeclaration(None=@n44) {name=LiquidCoordinates, nodeColor=<rgba 0 1 0.1094 1>}
  n61 = Subtract(A=Subtract(A=@n52, B=@n3) {constB=0}, B=@n24) {constB=0}
  n62 = Power(Base=Saturate(None=Divide(A=@n61) {constB=0.015})) {constExponent=1}
  n63 = Saturate(None=@n4)
  n64 = Power(Base=Saturate(None=Divide(A=@n52, B=@n3))) {constExponent=3}
  n65 = LinearInterpolate(A=Power(Base=@n63) {constExponent=3}, B=Multiply(A=@n63, B=OneMinus(None=@n64)), Alpha=NamedRerouteUsage())
  n66 = ScalarParameter() {defaultValue=1, parameterName=Foam Intensity}
  n67 = NamedRerouteDeclaration(None=Saturate(None=Multiply(A=@n65, B=@n66))) {name=FoamTop, nodeColor=<rgba 0.4297 0 1 1>}
  n68 = OneMinus(None=Saturate(None=Add(A=@n62, B=@n67)))
  n69 = Divide(A=NamedRerouteUsage(), B=@n24) {constB=0.02}
  n70 = Multiply(A=OneMinus(None=Saturate(None=@n69)))
  NamedRerouteDeclaration(None=LinearInterpolate(Alpha=Subtract(A=@n68, B=@n70) {constA=0}) {constB=0.333}) {name=Miniscus, nodeColor=<rgba 0 1 0.5313 1>}
  NamedRerouteDeclaration(None=@n30) {name=Emissive, nodeColor=<rgba 0.6875 1 0 1>}
  n71 = ComponentMask(None=NamedRerouteUsage()) {g=true, r=true}
  n72 = Constant2Vector() {g=0.01, r=0.06}
  n73 = ScalarParameter() {defaultValue=1, parameterName=Bubbles_Tile}
  n74 = Multiply(A=Add(A=Multiply(A=@n71, B=@n72) {constB=0.05}, B=NamedRerouteUsage()), B=Multiply(A=@n73, B=Constant2Vector() {g=1.5, r=2}) {constA=2})
  n75 = NamedRerouteUsage()
  n76 = Multiply(A=@n73, B=Constant2Vector() {g=2})
  n77 = TextureSample(UVs=BumpOffset(Coordinate=Panner(Coordinate=@n74, Time=@n75, Speed=@n76) {speedY=3}, Height=NamedRerouteUsage(), HeightRatioInput=Constant() {r=0.66}) {heightRatio=0.6}) {
   samplerSource=SSM_Wrap_WorldGroupSettings
   samplerType=SAMPLERTYPE_Normal
   texture=/Game/Characters/Player/Male/Large/Bodies/M_LRG_Slurp_Monster/Textures/T_SlurpBubble_N
  }
  NamedRerouteDeclaration(None=@n77.RGB) {name=Bubbles Normal, nodeColor=<rgba 0.3359 1 0 1>}
  n78 = TextureSampleParameter2D() {parameterName=Normal, samplerType=SAMPLERTYPE_Normal, texture=&FlatNormalBlue}
  n79 = NamedRerouteUsage()
  n80 = LinearInterpolate(A=NamedRerouteUsage(), B=@n78.RGB, Alpha=@n79)
  n81 = NamedRerouteUsage()
  n82 = NamedRerouteUsage()
  n83 = NamedRerouteUsage()
  n84 = Multiply(A=Constant2Vector() {g=0.2, r=0.3}) {constB=2}
  n85 = TextureSampleParameter2D(UVs=Panner(Coordinate=Multiply(A=@n83, B=@n84), Time=NamedRerouteUsage()) {speedY=-0.7}) {parameterName=Normal_Top, samplerType=SAMPLERTYPE_Normal, texture=/FortWater/Textures/02_WaterFoam_Veroni_base_N}
  n86 = NamedRerouteUsage()
  n87 = OneMinus(None=NamedRerouteUsage())
  n88 = LinearInterpolate(
   A=LinearInterpolate(
    A=LinearInterpolate(A=@n78.RGB, B=@n81, Alpha=@n82)
    B=NamedRerouteDeclaration(None=@n85.RGB) {name=Water Turbulence, nodeColor=<rgba 1 0 0.8203 1>}
    Alpha=Multiply(A=@n86, B=@n87)
   )
   B=Multiply(A=NamedRerouteUsage()) {constB=-1}
   Alpha=@n79
  )
  n89 = NamedRerouteUsage()
  NamedRerouteDeclaration(None=Normalize(VectorInput=LinearInterpolate(A=@n80, B=@n88, Alpha=@n89))) {name=Pixel Normals, nodeColor=<rgba 0 0.8125 1 1>}
  NamedRerouteDeclaration(None=DotProduct(A=VertexNormalWS(), B=CameraVectorWS())) {name=PixelBumpOffset, nodeColor=<rgba 0.0312 0 1 1>}
  n90 = VertexNormalWS()
  n91 = Multiply(A=OneMinus(None=Fresnel(Normal=@n90) {baseReflectFraction=0, exponent=2}), B=NamedRerouteUsage())
  n92 = NamedRerouteUsage()
  n93 = Saturate(None=Add(A=NamedRerouteUsage(), B=Multiply(A=@n92) {constB=3}))
  NamedRerouteDeclaration(None=Multiply(A=LinearInterpolate(A=@n91, Alpha=@n93) {constB=0}, B=ScalarParameter() {defaultValue=1, parameterName=Bubbles Multiplier})) {name=Bubbles Mask, nodeColor=<rgba 0.7813 0 1 1>}
  n94 = CameraVectorWS()
  n95 = Normalize(VectorInput=PixelNormalWS())
  n96 = Saturate(None=Subtract(A=DotProduct(A=@n94, B=@n95)) {constB=0.15})
  n97 = Constant() {r=0.6366}
  n98 = Multiply(A=LinearInterpolate(Alpha=Power(Base=@n96, Exp=@n97) {constExponent=0.5}) {constA=1, constB=0}, B=@n95)
  n99 = LinearInterpolate(Alpha=NamedRerouteUsage()) {constB=0.2}
  n100 = ScalarParameter() {parameterName=Refraction Roughness}
  n101 = LinearInterpolate(
   A=SkyLightEnvMapSample(Direction=Subtract(A=@n98, B=@n94) {constA=0, constB=0}, Roughness=Add(A=@n99, B=@n100))
   B=Constant3Vector() {constant=<rgba 1 1 1 0>}
   Alpha=ScalarParameter() {defaultValue=1, parameterName=FrontEnd Lerp}
  )
  n102 = Desaturation(None=@n101)
  n103 = ScalarParameter() {defaultValue=1, parameterName=Refraction Saturation}
  n104 = AtmosphericLightColor()
  n105 = VectorParameter() {parameterName=FrontEnd Refract Color, defaultValue=<rgba 1 0 0 1>}
  n106 = ScalarParameter() {defaultValue=1, parameterName=FrontEnd Lerp}
  n107 = Multiply(
   A=VectorParameter() {parameterName=Color, defaultValue=<rgba 0.7917 0.0966 0.1517 1>}.RGB
   B=Multiply(A=LinearInterpolate(A=@n102, B=@n101, Alpha=@n103), B=LinearInterpolate(A=@n104, B=@n105.RGB, Alpha=@n106))
  )
  n108 = VectorParameter() {parameterName=Foam Color, defaultValue=<rgba 0.4031 0.486 0.6 0>}
  n109 = NamedRerouteUsage()
  NamedRerouteDeclaration(None=Multiply(A=LinearInterpolate(A=@n107, B=@n108.RGB, Alpha=@n109), B=ScalarParameter() {defaultValue=1, parameterName=Color Multiplier})) {name=Base Color, nodeColor=<rgba 1 0.5859 0 1>}
  NamedRerouteDeclaration(None=LinearInterpolate(
    A=ScalarParameter() {defaultValue=0.025, parameterName=Roughness}
    B=ScalarParameter() {defaultValue=0.3, parameterName=Foam Roughness}
    Alpha=NamedRerouteUsage()
   ) {desc=roughness}) {name=Roughness, nodeColor=<rgba 1 0 0.4453 1>}
  NamedRerouteDeclaration(None=LinearInterpolate(
    A=ScalarParameter() {defaultValue=0.95, parameterName=Specular}
    B=ScalarParameter() {defaultValue=0.5, parameterName=Foam Specular}
    Alpha=NamedRerouteUsage()
   ) {desc=roughness}) {name=Specular, nodeColor=<rgba 0 0.9766 1 1>}
  n110 = TextureSampleParameter2D(UVs=@n22) {parameterName=Normal_Top, samplerType=SAMPLERTYPE_Normal, texture=&02_WaterFoam_Veroni_base_N}
  n111 = MaterialFunctionCall(
   BaseNormal=TextureSampleParameter2D(UVs=@n20) {parameterName=Normal_Top, samplerType=SAMPLERTYPE_Normal, texture=&02_WaterFoam_Veroni_base_N}.RGB
   AdditionalNormal=Multiply(
    A=MaterialFunctionCall(X=@n110.G, Y=@n110.G, Z=@n110.G) {materialFunction=&MakeFloat3}.Result
    B=Constant3Vector() {constant=<rgba 1 -1 1 0>}
   )
  ) {materialFunction=&BlendAngleCorrectedNormals}
  n112 = OneMinus(None=ComponentMask(None=Normalize(VectorInput=@n111.Result)) {b=true})
  n113 = ComponentMask(None=NamedRerouteUsage()) {b=true}
  NamedRerouteDeclaration(None=Multiply(A=Multiply(A=@n112, B=@n113)) {constB=-0.25}) {name=Foam Height, nodeColor=<rgba 1 0 0.2109 1>}
  n114 = LinearInterpolate(
   A=Constant3Vector() {desc=Renormal, constant=<rgba -1 1 -1 0>}
   B=Constant3Vector() {desc=Renormal, constant=<rgba 1 -1 -1 0>}
   Alpha=NamedRerouteUsage()
  )
  n115 = Constant3Vector() {constant=<rgba 0 0 1 0>}
  n116 = NamedRerouteUsage()
  n117 = MaterialFunctionCall(
   BaseNormal=NamedRerouteUsage()
   AdditionalNormal=Multiply(A=Multiply(A=@n111.Result, B=@n114) {constB=-1}, B=LinearInterpolate(A=@n115, Alpha=@n116))
  ) {materialFunction=&BlendAngleCorrectedNormals}
  NamedRerouteDeclaration(None=Normalize(VectorInput=Transform(None=@n117.Result) {transformSourceType=TRANSFORMSOURCE_World, transformType=TRANSFORM_Tangent})) {name=Water Plane Normal, nodeColor=<rgba 0 0.6016 1 1>}
  NamedRerouteDeclaration() {name=SurfaceNoise, nodeColor=<rgba 1 0.8906 0 1>}
  NamedRerouteDeclaration(None=@n43.Result) {name=ObjectSpaceRotated, nodeColor=<rgba 0.8047 1 0 1>}
  n118 = DynamicParameter() {desc=WaterPlaneZ, defaultValue=<rgba 0 0 1 0>}
  NamedRerouteDeclaration(None=Normalize(VectorInput=MaterialFunctionCall(X=@n118.R, Y=@n118.R, Z=@n118.R) {materialFunction=&MakeFloat3}.Result)) {name=WaterPlaneZ, nodeColor=<rgba 1 0 0.1875 1>}
  n119 = ScalarParameter() {parameterName=Viscosity}
  NamedRerouteDeclaration(None=Multiply(A=Time(), B=LinearInterpolate(Alpha=@n119) {constA=1, constB=0.15})) {name=Viscosity, nodeColor=<rgba 0 1 0.5078 1>}
  n120 = WorldPosition()
  n121 = CameraPositionWS()
  n122 = DotProduct(A=VertexNormalWS(), B=Normalize(VectorInput=Subtract(A=@n120.XYZ, B=@n121)))
  n123 = SmoothStep(Max=Constant() {r=0.001}, Value=@n122)
  n124 = NamedRerouteUsage()
  n125 = ScalarParameter() {defaultValue=0.8, parameterName=WPO amount}
  n126 = MaterialFunctionCall(Offset Amount=Multiply(A=@n122, B=Multiply(A=@n124, B=@n125))) {materialFunction=/Engine/Functions/Engine_MaterialFunctions02/WorldPositionOffset/CameraOffset}
  NamedRerouteDeclaration(None=ShadowReplace(Default=Multiply(A=@n123, B=@n126.Clamped Camera Offset), Shadow=Constant3Vector())) {name=WPO, nodeColor=<rgba 0.3125 1 0 1>}
  NamedRerouteDeclaration(None=@n51) {name=Height, nodeColor=<rgba 0 1 0.8828 1>}
  NamedRerouteDeclaration(None=MaterialFunctionCall(X=@n38.R, Y=@n38.R, Z=@n38.R) {materialFunction=&MakeFloat3}.Result) {name=Velocity, nodeColor=<rgba 0.5 1 0 1>}
  NamedRerouteDeclaration(None=@n118.A) {name=Bubbles Intensity, nodeColor=<rgba 0 0.7891 1 1>}
  n127 = DynamicParameter() {parameterIndex=3, defaultValue=<rgba 0 0 1 0>}
  NamedRerouteDeclaration(None=@n127.A) {name=Fill, nodeColor=<rgba 0 1 0.0859 1>}
  NamedRerouteDeclaration(None=MaterialFunctionCall(X=@n127.R, Y=@n127.R, Z=@n127.R) {materialFunction=&MakeFloat3}.Result) {name=Dimensions, nodeColor=<rgba 0 0.0994 1 1>}
  n128 = NamedRerouteUsage()
  NamedRerouteDeclaration(None=Multiply(A=NamedRerouteUsage(), B=Multiply(A=@n128, B=@n46))) {name=Center, nodeColor=<rgba 1 0 0.0937 1>}
  NamedRerouteUsage()
  NamedRerouteUsage()
  NamedRerouteUsage()
  NamedRerouteUsage()
  NamedRerouteUsage()
  NamedRerouteUsage()
  NamedRerouteUsage()
  NamedRerouteUsage()
  NamedRerouteUsage()
  NamedRerouteUsage()
  NamedRerouteUsage()
  NamedRerouteUsage()
  ObjectBounds()
  Saturate(None=@n53) {desc=Opacity}
  Subtract(A=@n40, B=Transform(None=VectorParameter() {parameterName=CenterOffset}.RGB) {transformSourceType=TRANSFORMSOURCE_Local})
  TextureSample(Tex=TextureObject() {samplerType=SAMPLERTYPE_Masks, texture=/Game/Characters/Player/Male/Large/Bodies/M_LRG_Slurp_Monster/Textures/T_SlurpBubble})
  Time()
  VectorParameter() {parameterName=Color, defaultValue=<rgba 0.7917 0.0966 0.1517 1>}
  VertexNormalWS()
 ]
}
