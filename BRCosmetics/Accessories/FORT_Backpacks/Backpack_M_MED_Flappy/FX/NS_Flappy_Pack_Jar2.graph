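# Niagara effect for the Flappy backpack cosmetic, combining a sloshing jar liquid with drool and drips.
# Two near-identical liquid emitters, chosen by gameplay or front end, use a spring pendulum to orient the water plane and motion-triggered timelines for fizz.
# Drool meshes hang from skeletal-mesh sockets as pendulums, passing pivot and tip to material parameters.
# Socket-attached Drip_A sprites emit death events, likely spawning falling Drip_B drops.
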NiagaraSystem /BRCosmetics/Accessories/FORT_Backpacks/Backpack_M_MED_Flappy/FX/NS_Flappy_Pack_Jar2
dependencies {
 usedDataInterfaces [
  &NiagaraDataInterfaceCurve
  &NiagaraDataInterfaceMeshRendererInfo
  &NiagaraDataInterfaceArrayFloat
  &NiagaraDataInterfaceVectorCurve
  &NiagaraDataInterfaceVector2DCurve
  &NiagaraDataInterfaceArrayFloat2
  &NiagaraDataInterfaceArrayFloat3
  &NiagaraDataInterfaceArrayColor
  &NiagaraDataInterfaceColorCurve
  &NiagaraDataInterfaceArrayInt32
 ]
 usedDynamicInputs []
 usedModules [
  &SystemState
  &self:SystemSpawnScript.SystemScriptSource.SystemScriptGraph.NiagaraNodeAssignment_0.SetVariables_#1
  &self:SystemSpawnScript.SystemScriptSource.SystemScriptGraph.NiagaraNodeAssignment_1.SetVariables_#2
  &self:SystemSpawnScript.SystemScriptSource.SystemScriptGraph.NiagaraNodeAssignment_2.SetVariables_#3
 ]
}
emitters {
 Drip_A {
  emitterUpdateScript [
   &EmitterState(
    Life Cycle Mode=::NewEnumerator1
    Loop Duration=&UniformRangedFloat(Minimum=2, Maximum=4)
    Recalculate Duration Each Loop=true
    UseLoopDelay=false
   )
   &SpawnRate(Use Spawn Probability=false) {enabled=false}
   &SpawnBurst_Instantaneous(Spawn Count=1, Use Spawn Probability=false)
  ]
  particleSpawnScript [
   /Niagara/Modules/Spawn/Initialization/InitializeParticle(
    Color=<rgba 1 0 0.0997 1>
    Lifetime=&UniformRangedFloat(Minimum=1.8, Maximum=2.1)
    Module.Write Color=true
    Module.Write Lifetime=true
    Module.Write Mass=false
    Module.Write Material Random=true
    Module.Write Position=false
    Module.Write Scale=false
    Module.Write SpriteRotation=true
    Module.Write SpriteSize=true
    Position=&SimulationPosition()
    Sprite Rotation=180
    Sprite Size=&Vector2DFromFloat(Value=&MakeFloatFromVector(Vector=&UniformRangedVector(Minimum=<1.5 1 1>, Maximum=<2 1 1>), Channel=::NewEnumerator0))
   )
   &PartitionParticles(Number Of Partitions=3)
  ]
  particleUpdateScript [
   &ParticleState()
   &SkeletalMeshLocation(
    Bone / Socket Index=$Particles.PartitionParticles.Partition
    Inherit Velocity=false
    MaskSpawnGroup=false
    Mesh Sampling Type=::NewEnumerator5
    Sampled Position Offset=<0 0 3>
    Skeletal Mesh=$User.SkeletalMesh
    Socket Sampling Mode=::NewEnumerator5
   )
   &AccelerationForce(Acceleration=&VectorFromCurve(VectorCurve=DataInterface() {propertyValues {
       XCurve {keys [<at 0 0>,<at 0.9988 0>]}
       YCurve {keys [<at 0 0>,<at 0.9988 0>]}
       ZCurve {keys [
         {arriveTangent=-58.0901, interpMode=RCIM_Cubic, time=0, value=-0.1}
         {arriveTangent=-17.2708, interpMode=RCIM_Cubic, leaveTangent=-22613.7637, tangentMode=RCTM_User, time=0.3, value=-1}
        ]}
      }})) {enabled=false}
   &SolveForcesAndVelocity(Clamp Velocity=false, Limit Acceleration=false)
   &ScaleSpriteSize(Scale Factor=&Vector2DFromCurve(
     Vector2Curve=DataInterface() {propertyValues {
       XCurve {keys [
         {interpMode=RCIM_Cubic, tangentMode=RCTM_User, time=0, value=1.15}
         {arriveTangent=-0.8621, interpMode=RCIM_Cubic, leaveTangent=-0.8621, tangentMode=RCTM_User, time=0.6, value=1.14}
         <at 0.8256 0.6684>
         {arriveTangent=-0.351, interpMode=RCIM_Cubic, leaveTangent=-0.351, tangentMode=RCTM_User, time=0.869, value=0.6114}
         {interpMode=RCIM_Cubic, tangentMode=RCTM_User, time=0.9988, value=1}
        ]}
       YCurve {keys [
         {interpMode=RCIM_Cubic, tangentMode=RCTM_User, time=-0.0019, value=0.3593}
         {arriveTangent=2.8736, interpMode=RCIM_Cubic, leaveTangent=2.8736, tangentMode=RCTM_User, time=0.6, value=0.99}
         {arriveTangent=2.025, interpMode=RCIM_Cubic, leaveTangent=2.025, tangentMode=RCTM_User, time=0.87, value=1.8}
         {interpMode=RCIM_Cubic, tangentMode=RCTM_User, time=1, value=1.8}
        ]}
      }}
     ScaleCurve=<1.2 3>
    ))
   &GenerateDeathEvent(
    Gap Correction Amount=1.5
    Position To Send:Vector3f=/Niagara/DynamicInputs/Transforms/TransformPosition(
     Position=&Add_Vector(A=Linked() {linkedVariable {name=Particles.Position, type=NiagaraPosition}}, B=<0 0 -3>)
     Source Space=::Local
     Destination Space=::World
    )
   )
   &ScaleColor(
    Scale Alpha=&FloatFromCurve(FloatCurve=DataInterface() {propertyValues {Curve {keys [<at 0 0>,<at 0.1 1>,<at 0.9 1>,<at 1.0011 0.002>]}}})
    ScaleA=true
    ScaleRGB=true
   )
   &CameraOffset(Camera Offset Amount=5) {enabled=false}
  ]
  rendererClasses [&NiagaraSpriteRendererProperties]
  renderers [{rendererClass=&NiagaraSpriteRendererProperties}]
 }
 Drip_B {
  emitterUpdateScript [&EmitterState(Life Cycle Mode=::NewEnumerator1, Loop Duration=1, UseLoopDelay=false)]
  particleSpawnScript [
   &InitializeParticle(
    Color=<rgba 1 1 1 0.35>
    Lifetime=&UniformRangedFloat(Minimum=0.7, Maximum=1)
    Mass=&UniformRangedFloat(Minimum=0.75, Maximum=1.75)
    Module.Write Color=true
    Module.Write Lifetime=true
    Module.Write Mass=false
    Module.Write Material Random=true
    Module.Write Position=true
    Module.Write Scale=false
    Module.Write SpriteRotation=false
    Module.Write SpriteSize=true
    Sprite Size=&Vector2DFromFloat(Value=&MakeFloatFromVector(Vector=&UniformRangedVector(Minimum=<3.5 1 1>, Maximum=<4.5 1 1>), Channel=::NewEnumerator0))
   )
   &InheritVelocity(Module.Limit Speed=false, Velocity Scale=<0.5 0.5 0.5>)
  ]
  particleUpdateScript [
   &CameraOffset(Camera Offset Amount=6) {enabled=false}
   &ParticleState()
   &AccelerationForce(Acceleration=&VectorFromCurve(VectorCurve=DataInterface() {propertyValues {
       XCurve {keys [<at 0 0>,<at 0.9989 0>]}
       YCurve {keys [<at 0 0>,<at 0.9989 0>]}
       ZCurve {keys [
         {arriveTangent=-15.5802, interpMode=RCIM_Cubic, leaveTangent=-8.2205, tangentMode=RCTM_User, time=0, value=0}
         {arriveTangent=-962.5264, interpMode=RCIM_Cubic, leaveTangent=-1066.8696, tangentMode=RCTM_User, time=0.3, value=-350}
        ]}
      }}))
   &SolveForcesAndVelocity(Clamp Velocity=false, Limit Acceleration=false)
   &ScaleSpriteSize(Scale Factor=&Vector2DFromCurve(
     Vector2Curve=DataInterface() {propertyValues {
       XCurve {keys [
         {arriveTangent=5.9728, interpMode=RCIM_Cubic, leaveTangent=5.9728, tangentMode=RCTM_User, time=0, value=0.1}
         {arriveTangent=-0.1, interpMode=RCIM_Cubic, leaveTangent=-0.1, tangentMode=RCTM_User, time=0.18, value=0.99}
         {interpMode=RCIM_Cubic, tangentMode=RCTM_User, time=1, value=0}
        ]}
       YCurve {keys [
         {arriveTangent=6.2715, interpMode=RCIM_Cubic, leaveTangent=6.2715, tangentMode=RCTM_User, time=0, value=0.1}
         {arriveTangent=-0.1001, interpMode=RCIM_Cubic, leaveTangent=-0.1001, tangentMode=RCTM_User, time=0.18, value=0.99}
         {interpMode=RCIM_Cubic, tangentMode=RCTM_User, time=0.9989, value=0}
        ]}
      }}
     ScaleCurve=<1 1>
    ))
   /SetVariables(Particles.PositionAlt:NiagaraPosition=&Add_Vector(
     A=Linked() {linkedVariable {name=Particles.Position, type=NiagaraPosition}}
     B=&RandomRangeVector(Minimum=<0 0.5 0>, Maximum=<0 -0.5 -4>)
    )) {isSetParameters=true}
  ]
  rendererClasses [&NiagaraSpriteRendererProperties]
  renderers [
   {rendererClass=&NiagaraSpriteRendererProperties}
   {rendererClass=&NiagaraSpriteRendererProperties, rendererIndex=1}
  ]
 }
 Drool {
  emitterUpdateScript [
   &EmitterState(
    Life Cycle Mode=::NewEnumerator1
    Loop Duration=&RandomRangeFloat(Minimum=2, Maximum=3)
    UseLoopDelay=false
   )
   &SpawnBurst_Instantaneous(Spawn Count=1, Use Spawn Probability=false)
  ]
  particleSpawnScript [
   /SetVariables(Particles.PendulumDrag=5, Particles.PendulumLengthScale=0.1, Particles.PendulumMaskFalloff=1.5, Particles.PendulumRigidity=0.2) {isSetParameters=true}
   &InitializeParticle(
    Lifetime=&UniformRangedFloat(Minimum=2, Maximum=3.5)
    Mesh Scale=&Multiply_Vector(A=<1 1 1>, B=&MakeVector(X=0.08, Y=0.08, Z=$Particles.PendulumLengthScale))
    Module.Write Color=true
    Module.Write Lifetime=true
    Module.Write Mass=true
    Module.Write Material Random=true
    Module.Write Position=true
    Module.Write Scale=true
    Module.Write SpriteRotation=false
    Module.Write SpriteSize=true
    Sprite Size=<1 1>
   )
   /SetVariables(Particles.RandomOffset:Vector3f=/Niagara/DynamicInputs/Random/RandomVector(
     Vector Scale:NiagaraFloat=&UniformRangedFloat(Minimum=25, Maximum=202.999)
     Constraint Vector:Vector3f=<0 0 1>
     Randomness Mode:ENiagaraRandomnessMode=::NewEnumerator0
     Random Seed:NiagaraInt32=0
     Fixed Random Seed:NiagaraBool=false
    )) {enabled=false, isSetParameters=true}
  ]
  particleUpdateScript [
   &SkeletalMeshLocation(
    Apply To Particle Position=false
    Bone / Socket Index=&SelectIntFromArray(
     Array Sampling Mode=::NewEnumerator1
     Direct Array Index=?
     Int Selection Array=DataInterface() {propertyValues {IntData [3,3]}}
     Random Seed=0
    )
    Inherit Velocity=false
    MaskSpawnGroup=false
    Mesh Sampling Type=::NewEnumerator5
    Skeletal Mesh=$User.SkeletalMesh
    Socket Sampling Mode=::NewEnumerator5
   )
   &ParticleState()
   /SetVariables(Particles.MeshOrigin:Vector3f=&Add_Vector(
     A=$Output.SkeletalMeshLocation.SampledPosition
     B=&MakeVector(
      X=0
      Y=0
      Z=&FloatFromCurve(
       FloatCurve=DataInterface() {propertyValues {Curve {keys [
           <at 0 0>
           {arriveTangent=1.5112, interpMode=RCIM_Cubic, leaveTangent=1.5112, tangentMode=RCTM_User, time=0.7, value=0}
           {arriveTangent=4.7214, interpMode=RCIM_Cubic, leaveTangent=4.7214, tangentMode=RCTM_User, time=1, value=1}
          ]}}}
       Scale Curve=-170
      )
     )
    )) {isSetParameters=true}
   &Drag(Drag=$Particles.PendulumDrag, Use Linear Drag=true, Use Rotational Drag=false)
   &GravityForce(Gravity=<0 0 -900>)
   &PendulumSetup(
    Module.Enable Spring Driver=false
    Pendulum Length:NiagaraFloat=&Multiply_Float(A=100, B=$Particles.PendulumLengthScale)
    Pendulum Pivot:Vector3f=$Particles.MeshOrigin
    "Spring Driven Constraint (Tightness)"=3
   )
   &SolveForcesAndVelocity(Clamp Velocity=false, Limit Acceleration=false)
   &PendulumConstraint()
   &CalculateAccurateVelocity()
   /SetVariables(Particles.MeshFacing:Vector3f=&NormalizeVector(Vector To Normalize=&Subtract_Vector(A=$Particles.MeshOrigin, B=Linked() {linkedVariable {name=Particles.Position, type=NiagaraPosition}}))) {isSetParameters=true}
   /Niagara/Modules/Update/Orientation/V2/OrientMeshToVector(
    Look At Direction:Vector3f=&Multiply_VectorByFloat(Vector=$Particles.MeshFacing, Float=1)
    Override Quat=false
    Reference Axis Vector:Vector3f=<0 0 1>
    UseOptionalSideAxis=false
    UseOptionalUpAxis=false
   )
   &DynamicMaterialParameters(
    Index 2 Param 1=HlslExpression() {hlslExpression=Particles.MeshOrigin.x}
    Index 2 Param 2=HlslExpression() {hlslExpression=Particles.MeshOrigin.y}
    Index 2 Param 3=HlslExpression() {hlslExpression=Particles.MeshOrigin.z}
    Index 2 Param 4=$Particles.PendulumMaskFalloff
    Index 3 Param 1=&MakeFloatFromVector(Vector=Linked() {linkedVariable {name=Particles.Position, type=NiagaraPosition}}, Channel=::NewEnumerator0)
    Index 3 Param 2=&MakeFloatFromVector(Vector=Linked() {linkedVariable {name=Particles.Position, type=NiagaraPosition}}, Channel=::NewEnumerator1)
    Index 3 Param 3=&MakeFloatFromVector(Vector=Linked() {linkedVariable {name=Particles.Position, type=NiagaraPosition}}, Channel=::NewEnumerator2)
    Index 3 Param 4=$Particles.PendulumRigidity
    Module.Param0WriteEnabled2=true
    Module.Param0WriteEnabled3=true
    Module.Param1WriteEnabled2=true
    Module.Param1WriteEnabled3=true
    Module.Param2WriteEnabled2=true
    Module.Param2WriteEnabled3=true
    Module.Param3WriteEnabled2=true
    Module.Param3WriteEnabled3=true
    Write Parameter Index 0=false
    Write Parameter Index 2=true
    Write Parameter Index 3=true
   )
   &ScaleMeshSize(Scale Factor=&MakeVector(
     X=&FloatFromCurve(FloatCurve=DataInterface() {propertyValues {Curve {keys [<at 0 0.2>,<at 0.7 1>]}}})
     Y=&FloatFromCurve(FloatCurve=DataInterface() {propertyValues {Curve {keys [<at 0 0.2>,<at 0.7 1>]}}})
     Z=&FloatFromCurve(FloatCurve=DataInterface() {propertyValues {Curve {keys [
          <at 0 0.25 RCIM_Cubic>
          <at 0.7 0.8055 RCIM_Cubic>
          {arriveTangent=-1.5491, interpMode=RCIM_Cubic, leaveTangent=-1.5491, tangentMode=RCTM_User, time=1, value=0.25}
         ]}}})
    ))
   &GenerateDeathEvent() {enabled=false}
   &GenerateLocationEvent(Event Send Rate=1, Module.Use Event Send Rate=true) {enabled=false}
  ]
  rendererClasses [&NiagaraMeshRendererProperties, &NiagaraSpriteRendererProperties]
  renderers [
   {rendererClass=&NiagaraMeshRendererProperties, rendererIndex=0}
   {rendererClass=&NiagaraSpriteRendererProperties, rendererIndex=1}
  ]
 }
 Gumball_Liquid_Water002 {
  emitterSpawnScript [
   /SetVariables() {enabled=false, isSetParameters=true, module=SetVariables_#4}
   /SetVariables() {enabled=false, isSetParameters=true, module=SetVariables_#5}
  ]
  emitterUpdateScript [
   &EmitterState(
    Inactive Response=::NewEnumerator1
    Life Cycle Mode=::NewEnumerator1
    Loop Behavior=::NewEnumerator1
    Loop Duration Mode=::NewEnumerator1
    UseLoopDelay=false
   )
   &SpawnBurst_Instantaneous(
    Spawn Probability=&FN_FrontEndFloatModifier(Gameplay=1, Front End=0)
    Use Spawn Probability=true
    UseLoopCountLimit=true
   )
   /SetVariables(Emitter.Orientation:Quat4f=$Engine.Owner.Rotation) {isSetParameters=true, module=SetVariables_#6}
   /SetVariables() {isSetParameters=true, module=SetVariables_#7}
   /SetVariables(
    Emitter.RotationVelocity:Vector3f=&Divide_Vector(
     A:Vector3f=&Subtract_Vector(
      A=&RotateVectorByQuaternion(Vector To Rotate=<0 0 1>, Quaternion=$Emitter.Orientation)
      B=&RotateVectorByQuaternion(Vector To Rotate=<0 0 1>, Quaternion=$Emitter.Previous.Orientation)
     )
     B:Vector3f=&VectorFromFloat(Value=$Engine.DeltaTime)
    )
    Emitter.Velocity:Vector3f=&Divide_Vector(
     A:Vector3f=&Subtract_Vector(A=$Emitter.Position, B=$Emitter.Previous.Position)
     B:Vector3f=&VectorFromFloat(Value=$Engine.DeltaTime)
    )
   ) {isSetParameters=true, module=SetVariables_#8}
   /SetVariables(Emitter.ColorA:LinearColor=<rgba 0.0936 0.3021 0.0302 1>, Emitter.ColorB:LinearColor=<rgba 0.0936 0.3021 0.0302 1>) {isSetParameters=true, module=SetVariables_#9}
  ]
  particleSpawnScript [
   /Niagara/Modules/Spawn/Initialization/V2/InitializeParticle(
    Color Mode=::NewEnumerator0
    Mesh Scale Mode=::NewEnumerator3
    Sprite Size Mode=::NewEnumerator3
    UsePositionOffset=false
    Write Material Random=true
   )
   &self:Initialize_Liquid(
    Liquid Fill Amount=0.6
    Liquid Mesh Renderer=DataInterface() {propertyValues {MeshRenderer=&self:Gumball_Liquid_Water002_1.NiagaraMeshRendererProperties_0}}
   ) {module=Initialize_Liquid}
  ]
  particleUpdateScript [
   &ParticleState(Kill Particles When Lifetime Has Elapsed=false)
   /SetVariables(Particles.Position_Origin:NiagaraPosition=/Niagara/DynamicInputs/Helpers/SimulationPosition()) {isSetParameters=true, module=SetVariables_#10}
   &PendulumSetup(
    Module.Enable Spring Driver=true
    Pendulum Length=5
    Pendulum Pivot:NiagaraPosition=$Particles.Position_Origin
    "Spring Driven Constraint (Tightness)"=0.1
   )
   &InheritSourceMovement(
    Applied Position Scale=&Multiply_VectorByFloat(Vector=<1 1 1>, Float=1)
    ApplyForce=false
    ApplyPosition=true
    ApplyPositionFalloff=false
    Source Velocity:Vector3f=$Emitter.RotationVelocity
   ) {enabled=false}
   &InheritVelocity(
    Inherited Velocity Amount Scale=&Multiply_VectorByFloat(Vector=<1 1 1>, Float=-0.1)
    Speed Limit=true
    UseSpeedThreshold=true
   )
   &Drag(Drag=1.5, Use Linear Drag=true, Use Rotational Drag=true)
   &SolveForcesAndVelocity(Clamp Velocity=true, Limit Acceleration=false, Speed Limit=15)
   &PendulumConstraint()
   &CalculateAccurateVelocity()
   &Timeline(
    Play=&SetBoolByFloatComparison(
     A=&Abs_Float(Float:NiagaraFloat=&VectorLength(Vector=&Add_Vector(A=$Emitter.Velocity, B=$Emitter.RotationVelocity), Safe Length Calculation=false))
     B=$User.Bubbles Min Speed
    )
    Play Rate=3
    Timeline End=10
   )
   &Timeline(Play=&SetBoolByFloatComparison(
     A=&Abs_Float(Float:NiagaraFloat=&VectorLength(Vector=$Particles.Velocity, Safe Length Calculation=false))
     B=$User.Bubbles Min Speed
    )) {module=Timeline001}
   &OrientMeshToVector(
    Look At Direction:Vector3f=&Subtract_Vector(
     A=Linked() {linkedVariable {name=Particles.Position_Origin, type=NiagaraPosition}}
     B=Linked() {linkedVariable {name=Particles.Position, type=NiagaraPosition}}
    )
    Override Quat=false
    Reference Axis Vector:Vector3f=<0 0 1>
    UseOptionalSideAxis=false
    UseOptionalUpAxis=false
   )
   /SetVariables(
    Particles.FizzAmount:NiagaraFloat=$Particles.Timeline.NormalizedTime
    Particles.WaterPlaneUp:Vector3f=&Subtract_Vector(A=&SimulationPosition(), B=Linked() {linkedVariable {name=Particles.Position, type=NiagaraPosition}})
   ) {isSetParameters=true, module=SetVariables_#11}
   /SetVariables(
    Particles.WaterPlaneZ:Vector4f=&MakeFloat4FromVectorAndFloat(
     "Vector (RGB)"=&ReturnMeshOrientationAxis(Mesh Orientation Axis=::ZAxis, Mesh Orientation Quat=$Particles.MeshOrientation)
     "Float (Alpha)"=$Particles.FizzAmount
    )
    Particles.WaterVelocity:Vector4f=&MakeFloat4FromVectorAndFloat("Vector (RGB)"=$Particles.Velocity, "Float (Alpha)"=$Particles.Timeline001.NormalizedTime)
   ) {isSetParameters=true, module=SetVariables_#12}
   /SetVariables(Particles.WaterPlaneTangent:Vector4f=&MakeFloat4FromVectorAndFloat(
     "Vector (RGB)"=&ReturnMeshOrientationAxis(Mesh Orientation Axis=::XAxis, Mesh Orientation Quat=$Particles.MeshOrientation)
     "Float (Alpha)"=1
    )) {isSetParameters=true, module=SetVariables_#13}
  ]
  rendererClasses [&NiagaraMeshRendererProperties]
  renderers [
   {rendererClass=&NiagaraMeshRendererProperties, rendererIndex=0}
   {rendererClass=&NiagaraMeshRendererProperties, rendererIndex=1}
  ]
 }
 Gumball_Liquid_Water_FE {
  emitterSpawnScript [
   /SetVariables() {enabled=false, isSetParameters=true, module=SetVariables_#4}
   /SetVariables() {enabled=false, isSetParameters=true, module=SetVariables_#5}
  ]
  emitterUpdateScript [
   &EmitterState(
    Inactive Response=::NewEnumerator1
    Life Cycle Mode=::NewEnumerator1
    Loop Behavior=::NewEnumerator1
    Loop Duration Mode=::NewEnumerator1
    UseLoopDelay=false
   )
   &SpawnBurst_Instantaneous(
    Spawn Probability=&FN_FrontEndFloatModifier(Gameplay=0, Front End=1)
    Use Spawn Probability=true
    UseLoopCountLimit=true
   )
   /SetVariables(Emitter.Orientation:Quat4f=$Engine.Owner.Rotation) {isSetParameters=true, module=SetVariables_#6}
   /SetVariables() {isSetParameters=true, module=SetVariables_#7}
   /SetVariables(
    Emitter.RotationVelocity:Vector3f=&Divide_Vector(
     A:Vector3f=&Subtract_Vector(
      A=&RotateVectorByQuaternion(Vector To Rotate=<0 0 1>, Quaternion=$Emitter.Orientation)
      B=&RotateVectorByQuaternion(Vector To Rotate=<0 0 1>, Quaternion=$Emitter.Previous.Orientation)
     )
     B:Vector3f=&VectorFromFloat(Value=$Engine.DeltaTime)
    )
    Emitter.Velocity:Vector3f=&Divide_Vector(
     A:Vector3f=&Subtract_Vector(A=$Emitter.Position, B=$Emitter.Previous.Position)
     B:Vector3f=&VectorFromFloat(Value=$Engine.DeltaTime)
    )
   ) {isSetParameters=true, module=SetVariables_#8}
   /SetVariables(Emitter.ColorA:LinearColor=<rgba 0.0936 0.3021 0.0302 1>, Emitter.ColorB:LinearColor=<rgba 0.0936 0.3021 0.0302 1>) {isSetParameters=true, module=SetVariables_#9}
  ]
  particleSpawnScript [
   &InitializeParticle(
    Color Mode=::NewEnumerator0
    Mesh Scale Mode=::NewEnumerator3
    Sprite Size Mode=::NewEnumerator3
    UsePositionOffset=false
    Write Material Random=true
   )
   &self:Initialize_Liquid(
    Liquid Fill Amount=0.6
    Liquid Mesh Renderer=DataInterface() {propertyValues {MeshRenderer=&self:Gumball_Liquid_Water_FE_2.NiagaraMeshRendererProperties_1}}
   ) {module=Initialize_Liquid}
  ]
  particleUpdateScript [
   &ParticleState(Kill Particles When Lifetime Has Elapsed=false)
   /SetVariables(Particles.Position_Origin:NiagaraPosition=&SimulationPosition()) {isSetParameters=true, module=SetVariables_#10}
   &PendulumSetup(
    Module.Enable Spring Driver=true
    Pendulum Length=5
    Pendulum Pivot:NiagaraPosition=$Particles.Position_Origin
    "Spring Driven Constraint (Tightness)"=0.1
   )
   &InheritSourceMovement(
    Applied Position Scale=&Multiply_VectorByFloat(Vector=<1 1 1>, Float=1)
    ApplyForce=false
    ApplyPosition=true
    ApplyPositionFalloff=false
    Source Velocity:Vector3f=$Emitter.RotationVelocity
   ) {enabled=false}
   &InheritVelocity(
    Inherited Velocity Amount Scale=&Multiply_VectorByFloat(Vector=<1 1 1>, Float=-0.1)
    Speed Limit=true
    UseSpeedThreshold=true
   )
   &Drag(Drag=1.5, Use Linear Drag=true, Use Rotational Drag=true)
   &SolveForcesAndVelocity(Clamp Velocity=true, Limit Acceleration=false, Speed Limit=15)
   &PendulumConstraint()
   &CalculateAccurateVelocity()
   &Timeline(
    Play=&SetBoolByFloatComparison(
     A=&Abs_Float(Float:NiagaraFloat=&VectorLength(Vector=&Add_Vector(A=$Emitter.Velocity, B=$Emitter.RotationVelocity), Safe Length Calculation=false))
     B=$User.Bubbles Min Speed
    )
    Play Rate=3
    Timeline End=10
   )
   &Timeline(Play=&SetBoolByFloatComparison(
     A=&Abs_Float(Float:NiagaraFloat=&VectorLength(Vector=$Particles.Velocity, Safe Length Calculation=false))
     B=$User.Bubbles Min Speed
    )) {module=Timeline001}
   &OrientMeshToVector(
    Look At Direction:Vector3f=&Subtract_Vector(
     A=Linked() {linkedVariable {name=Particles.Position_Origin, type=NiagaraPosition}}
     B=Linked() {linkedVariable {name=Particles.Position, type=NiagaraPosition}}
    )
    Override Quat=false
    Reference Axis Vector:Vector3f=<0 0 1>
    UseOptionalSideAxis=false
    UseOptionalUpAxis=false
   )
   /SetVariables(
    Particles.FizzAmount:NiagaraFloat=$Particles.Timeline.NormalizedTime
    Particles.WaterPlaneUp:Vector3f=&Subtract_Vector(A=&SimulationPosition(), B=Linked() {linkedVariable {name=Particles.Position, type=NiagaraPosition}})
   ) {isSetParameters=true, module=SetVariables_#11}
   /SetVariables(
    Particles.WaterPlaneZ:Vector4f=&MakeFloat4FromVectorAndFloat(
     "Vector (RGB)"=&ReturnMeshOrientationAxis(Mesh Orientation Axis=::ZAxis, Mesh Orientation Quat=$Particles.MeshOrientation)
     "Float (Alpha)"=$Particles.FizzAmount
    )
    Particles.WaterVelocity:Vector4f=&MakeFloat4FromVectorAndFloat("Vector (RGB)"=$Particles.Velocity, "Float (Alpha)"=$Particles.Timeline001.NormalizedTime)
   ) {isSetParameters=true, module=SetVariables_#12}
   /SetVariables(Particles.WaterPlaneTangent:Vector4f=&MakeFloat4FromVectorAndFloat(
     "Vector (RGB)"=&ReturnMeshOrientationAxis(Mesh Orientation Axis=::XAxis, Mesh Orientation Quat=$Particles.MeshOrientation)
     "Float (Alpha)"=1
    )) {isSetParameters=true, module=SetVariables_#13}
  ]
  rendererClasses [&NiagaraMeshRendererProperties]
  renderers [{rendererClass=&NiagaraMeshRendererProperties, rendererIndex=0}]
 }
}
summary {
 systemName=NS_Flappy_Pack_Jar2
 emitters [
  {emitterName=Gumball_Liquid_Water002, rendererClasses [&NiagaraMeshRendererProperties]}
  {emitterName=Drip_A, rendererClasses [&NiagaraSpriteRendererProperties]}
  {emitterName=Drip_B, rendererClasses [&NiagaraSpriteRendererProperties]}
  {emitterName=Drool, rendererClasses [&NiagaraMeshRendererProperties, &NiagaraSpriteRendererProperties]}
  {emitterName=Gumball_Liquid_Water_FE, rendererClasses [&NiagaraMeshRendererProperties]}
 ]
 userVariables [
  User.bIsFrontEnd {type=NiagaraBool}
  User.bIsFrontEndPreview {type=NiagaraBool}
  User.BowlRadius {type=NiagaraFloat, defaultValue=14, description=""}
  User.Bubbles Min Speed {type=NiagaraFloat, defaultValue=20, description=""}
  User.Liquid Center Offset {type=Vector3f, description="", defaultValue {struct=Vector3f, value=<0 0 0>}}
  User.Liquid Dimensions {type=Vector3f, description="", defaultValue {struct=Vector3f, value=<14 14 14>}}
  User.Liquid Fullness {type=NiagaraFloat, defaultValue=0.75, description=""}
  User.SkeletalMesh {
   type=NiagaraDataInterfaceSkeletalMesh
   defaultValue {
    struct=NiagaraExt_VariableValue_DataInterface
    value {dataInterface=&self:NiagaraDataInterfaceSkeletalMesh_0}
   }
  }
 ]
}
